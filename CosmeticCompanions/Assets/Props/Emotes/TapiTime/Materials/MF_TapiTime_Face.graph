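# Material function for the TapiTime emote prop that overlays up to three animated-style faces onto incoming material attributes.
# An RGB mask selects per-face index, UV offset, scale and rotation, so shared transformed UVs drive 3x3 flipbook lookups for face color and emissive.
# Center-scaled UV masks confine the blends, with a tinted eye glow, Fresnel-based face brightening and multipliers for base color, metallic, specular and roughness.
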MaterialFunction /CosmeticCompanions/Assets/Props/Emotes/TapiTime/Materials/MF_TapiTime_Face
graph {nodes [
  n1 = GetMaterialAttributes(None=NamedRerouteUsage()) {attributeGetTypes [
    #1
    #2
    #3
    #4
    #5
   ]}
  n2 = SetMaterialAttributes!(
   MaterialAttributes=@n1.MaterialAttributes
   BaseColor=Multiply(A=NamedRerouteUsage(), B=ScalarParameter() {defaultValue=1, parameterName=BaseColorMultiplier})
   EmissiveColor
   Metallic=Multiply(A=NamedRerouteUsage(), B=ScalarParameter() {defaultValue=1, parameterName=MetallicScale})
   Specular=Multiply(A=NamedRerouteUsage(), B=ScalarParameter() {defaultValue=1, parameterName=SpecularScale})
   Roughness=Multiply(A=NamedRerouteUsage(), B=ScalarParameter() {defaultValue=1, parameterName=RoughnessScale})
  ) {attributeSetTypes [
    #1
    #2
    #3
    #4
    #5
   ]}
  FunctionOutput(None=@n2)
  n3 = Clamp(None=ScalarParameter() {group=FaceParameters, parameterName=FaceIndex2}) {maxDefault=6}
  n4 = Clamp(None=ScalarParameter() {group=FaceParameters, parameterName=FaceIndex1}) {maxDefault=6}
  n5 = NamedRerouteUsage()
  n6 = ScalarParameter() {group=FaceParameters, parameterName=FaceIndex3}
  NamedRerouteDeclaration(None=LinearInterpolate(A=LinearInterpolate(A=@n3, B=@n4, Alpha=@n5), B=Clamp(None=@n6) {maxDefault=6}, Alpha=NamedRerouteUsage())) {name=FaceIndex, nodeColor=<rgba 0.4062 0 1 1>}
  n7 = VectorParameter() {group=FaceParameters, parameterName=Face1UVOffset_Face2UVOffset, defaultValue=<rgba 0.5924 0.32 0.6616 0.6944>}
  n8 = ComponentMask(None=@n7.RGBA) {a=true, b=true}
  n9 = ComponentMask(None=@n7.RGB) {g=true, r=true}
  n10 = NamedRerouteUsage()
  n11 = VectorParameter() {group=FaceParameters, parameterName=Face3UVOffset, defaultValue=<rgba 0.4128 0.584 0 0>}
  NamedRerouteDeclaration(None=LinearInterpolate(A=LinearInterpolate(A=@n8, B=@n9, Alpha=@n10), B=ComponentMask(None=@n11.RGB) {g=true, r=true}, Alpha=NamedRerouteUsage())) {name=FaceUVOffset, nodeColor=<rgba 0.9568 1 0.0147 1>}
  NamedRerouteDeclaration(None=@n1.Metallic) {name=In_Metallic, nodeColor=<rgba 1 0 0.2344 1>}
  NamedRerouteDeclaration(None=@n1.Specular) {name=In_Specular, nodeColor=<rgba 0 0.5547 1 1>}
  NamedRerouteDeclaration(None=@n1.Roughness) {name=In_Roughness, nodeColor=<rgba 0.875 1 0 1>}
  NamedRerouteDeclaration(None=FunctionInput() {inputName=In_Fresnel, inputType=FunctionInput_Scalar, sortPriority=1}) {name=In_Fresnel, nodeColor=<rgba 0.6406 1 0 1>}
  NamedRerouteDeclaration(None=FunctionInput() {inputName=In_TechArtMask, sortPriority=2}) {name=In_TechArtMask, nodeColor=<rgba 0 0.1686 1 1>}
  n12 = NamedRerouteUsage()
  n13 = NamedRerouteUsage()
  NamedRerouteDeclaration(None=Add(A=Add(A=@n12, B=@n13), B=NamedRerouteUsage())) {name=FaceMasks, nodeColor=<rgba 0 1 0.1328 1>}
  n14 = TextureSampleParameter2D() {
   group=FaceParameters
   parameterName=FaceIndexMask
   samplerType=SAMPLERTYPE_Masks
   texture=/CosmeticCompanions/Assets/Props/Emotes/TapiTime/Textures/T_TapiTime_Emote_FaceIndex
  }
  NamedRerouteDeclaration(None=@n14.R) {name=FaceIndex1Mask, nodeColor=<rgba 1 0 0 1>}
  n15 = NamedRerouteUsage()
  n16 = NamedRerouteUsage()
  n17 = Constant() {r=3}
  n18 = TextureCoordinate()
  n19 = NamedRerouteUsage()
  n20 = Rotator(
   Coordinate=Multiply(A=Subtract(A=@n18, B=@n19), B=NamedRerouteUsage())
   Time=MaterialFunctionCall(Multiplier=NamedRerouteUsage()) {materialFunction=&Pi}.Result
  ) {speed=1}
  n21 = TextureSampleParameter2D(UVs=MaterialFunctionCall("Animation  Phase (0-1)"=Divide(A=@n16) {constB=9}, Number of Rows=@n17, Number of Columns=@n17, UVs=@n20) {materialFunction=&FlipBook}.UVs) {
   group=FaceParameters
   parameterName=FaceFlipbookEmissive
   texture=/CosmeticCompanions/Assets/Props/Emotes/TapiTime/Textures/TapiTime_Emote_Prop_Emissive_Face
  }
  n22 = Multiply(
   A=NamedRerouteDeclaration(None=@n21.R) {name=EmissiveGlow, nodeColor=<rgba 1 0 0.9844 1>}
   B=VectorParameter() {parameterName=GlowAroundEyes, defaultValue=<rgba 1 0 0 0>}.RGB
  )
  n23 = MaterialFunctionCall(UVs=@n20) {materialFunction=&ScaleUVsByCenter}
  NamedRerouteDeclaration(None=LinearInterpolate(A=NamedRerouteUsage(), B=Add(A=@n15, B=@n22), Alpha=Multiply(A=@n21.R, B=@n23.0-1 mask))) {name=Out_Emissive, nodeColor=<rgba 1 0.0182 0.5642 1>}
  NamedRerouteDeclaration(None=@n14.G) {name=FaceIndex2Mask, nodeColor=<rgba 0 1 0 1>}
  NamedRerouteDeclaration(None=@n14.B) {name=FaceIndex3Mask, nodeColor=<rgba 0 0 1 1>}
  n24 = VectorParameter() {group=FaceParameters, parameterName=Face1UVScale_Face2UVScale, defaultValue=<rgba 4 4 4 4>}
  n25 = ComponentMask(None=@n24.RGBA) {a=true, b=true}
  n26 = ComponentMask(None=@n24.RGB) {g=true, r=true}
  n27 = NamedRerouteUsage()
  n28 = VectorParameter() {group=FaceParameters, parameterName=Face3UVScale, defaultValue=<rgba 4 4 0 0>}
  NamedRerouteDeclaration(None=LinearInterpolate(A=LinearInterpolate(A=@n25, B=@n26, Alpha=@n27), B=ComponentMask(None=@n28.RGB) {g=true, r=true}, Alpha=NamedRerouteUsage())) {name=FaceUVScale, nodeColor=<rgba 0 0.5547 1 1>}
  n29 = ScalarParameter() {defaultValue=1, group=FaceParameters, parameterName=Face2RotationAngle}
  n30 = ScalarParameter() {defaultValue=0.5, group=FaceParameters, parameterName=Face1RotationAngle}
  n31 = NamedRerouteUsage()
  NamedRerouteDeclaration(None=LinearInterpolate(
    A=LinearInterpolate(A=@n29, B=@n30, Alpha=@n31)
    B=ScalarParameter() {defaultValue=0.5, group=FaceParameters, parameterName=Face3RotationAngle}
    Alpha=NamedRerouteUsage()
   )) {name=FaceRotationAngle, nodeColor=<rgba 1 0.1691 0 1>}
  NamedRerouteDeclaration(None=FunctionInput() {inputName=MaterialAttributes, inputType=FunctionInput_MaterialAttributes}) {name=In_MaterialAttributes, nodeColor=<rgba 0.8047 0 1 1>}
  n32 = NamedRerouteUsage()
  n33 = Constant() {r=3}
  n34 = TextureSampleParameter2D(UVs=MaterialFunctionCall("Animation  Phase (0-1)"=Divide(A=@n32) {constB=9}, Number of Rows=@n33, Number of Columns=@n33, UVs=@n20) {materialFunction=&FlipBook}.UVs) {group=FaceParameters, parameterName=FaceFlipbook, texture=/CosmeticCompanions/Assets/Props/Emotes/TapiTime/Textures/TapiTime_Emote_Prop_Face}
  n35 = Multiply(
   A=@n34.A
   B=MaterialFunctionCall(UVs=@n20) {materialFunction=&ScaleUVsByCenter}.0-1 mask
  )
  n36 = LinearInterpolate(A=NamedRerouteUsage(), B=@n34.RGB, Alpha=@n35)
  n37 = ScalarParameter() {defaultValue=5, parameterName=FresnelExponentIn}
  n38 = Add(A=OneMinus(None=Fresnel(ExponentIn=@n37) {exponent=12})) {constB=0.1}
  n39 = ScalarParameter() {defaultValue=1, parameterName=FaceFresnelMultiplier}
  n40 = Multiply(
   A=@n36
   B=SmoothStep(
    Min=ScalarParameter() {parameterName=FaceFresnelSmoothStepMin}
    Max=ScalarParameter() {defaultValue=1, parameterName=FaceFresnelSmoothStepMax}
    Value=Multiply(A=@n38, B=@n39)
   ) {constMin=0.5}
  )
  NamedRerouteDeclaration(None=LinearInterpolate(A=@n36, B=Add(A=@n36, B=@n40), Alpha=NamedRerouteUsage())) {name=Out_BaseColor, nodeColor=<rgba 0 1 0.4844 1>}
  NamedRerouteDeclaration(None=@n1.BaseColor) {name=In_BaseColor, nodeColor=<rgba 1 0.1641 0 1>}
  NamedRerouteDeclaration(None=@n1.EmissiveColor) {name=In_EmissiveColor, nodeColor=<rgba 0 1 0.3672 1>}
  NamedRerouteDeclaration(None=@n35) {name=ScaledUVs, nodeColor=<rgba 1 0.0469 0 1>}
  NamedRerouteUsage()
 ]}
