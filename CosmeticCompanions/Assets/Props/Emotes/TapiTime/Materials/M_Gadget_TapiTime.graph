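# Character material for the TapiTime cosmetic companion prop, built on the shared 2023 character parent material functions.
# The PreFX parent function builds the material attributes, which pass through the TapiTime face function and then into the PostFX parent function.
# It uses material attributes with an expression-driven shading model and supports skeletal meshes, morph targets, mesh deformers and Nanite.
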Material /CosmeticCompanions/Assets/Props/Emotes/TapiTime/Materials/M_Gadget_TapiTime
graph {
 properties {
  bHasPixelAnimation=true
  bUseMaterialAttributes=true
  bUsedWithMeshDeformer=true
  bUsedWithMorphTargets=true
  bUsedWithNanite=true
  bUsedWithSkeletalMesh=true
  pixelDepthOffsetMode=PDOM_Legacy
  previewMesh=/CosmeticCompanions/Assets/Biped/TapiTime/Meshes/TapiTime_Biped
  shadingModel=MSM_FromMaterialExpression
  usedShadingModels=Default Lit | Subsurface | Clear Coat | Subsurface Profile | Cloth
 }
 nodes [
  n1 = MaterialFunctionCall() {materialFunction=&MF_CharacterParent_2023_PreFX}
  MaterialFunctionCall(
   Material=MaterialFunctionCall!(MaterialAttributes=@n1.Result, In_Fresnel=@n1.Result, In_TechArtMask=@n1.Result) {materialFunction=&/MF_TapiTime_Face}.Result
   M=@n1.M
   Pre Skinned Local Position=@n1.M
   Distance Blend=@n1.M
   Fresnel=@n1.M
  ) {materialFunction=&MF_CharacterParent_2023_PostFX}
 ]
}
