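# Physics asset for the TapiTime gadget, an emote prop in CosmeticCompanions.
# It defines three bodies on the bones ball_A, ball_B and ball_C, each approximated by a single capsule.
# The ball_A and ball_B capsules are nearly or fully spherical, while ball_C is a smaller, elongated, tilted capsule.
# No constraints link the bodies, so the asset defines no joints or limits between them.
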PhysicsAsset /CosmeticCompanions/Assets/Props/Emotes/TapiTime/Meshes/PA_TapiTime_Gadget
bodies {
 ball_A [{shapeType=Capsule, length=1.7213, radius=24.2592, center=<0 -1.3815 5.385>, rotation=<rot 89.6141 -90.0006 -0.0006>}]
 ball_B [{shapeType=Capsule, length=0, radius=22.3808, shapeName=ball_B_capsule, center=<0.0709 0.147 1.2342>, rotation=<rot 0 88.3691 -89.6331>}]
 ball_C [{shapeType=Capsule, length=13.4645, radius=9.0384, center=<-0.009 -1.7061 2.9696>, rotation=<rot 0 -0.2815 -51.3358>}]
}
constraints []
